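# MF_InverseProportional is a material function that remaps a grayscale scalar along an inverse-proportional (hyperbolic) curve.
# The value input is saturated, optionally inverted by a static switch (on by default), then used as the divisor of scale.
# Subtracting scale shifts the curve to start at zero by default, and a pow exponent shapes the result.
# The output is unclamped, and the divisor has no guard against zero.
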MaterialFunction /Game/Global/Materials/MaterialGym/Materials/MaterialElements/GrayscaleTransforms/InverseProportional/MF_InverseProportional
graph {nodes [
  n1 = FunctionInput() {bUsePreviewValueAsDefault=true, inputName=scale, inputType=FunctionInput_Scalar, sortPriority=1, previewValue=<0.02 0 0 1>}
  n2 = Saturate(None=FunctionInput() {bUsePreviewValueAsDefault=true, inputName=value, inputType=FunctionInput_Scalar})
  n3 = FunctionInput(Preview=StaticBool() {value=true}) {bUsePreviewValueAsDefault=true, inputName=invert value, inputType=FunctionInput_StaticBool, sortPriority=2, previewValue=<1 0 0 1>}
  n4 = Divide(A=@n1, B=StaticSwitch(True=OneMinus(None=@n2), False=@n2, Value=@n3))
  n5 = FunctionInput() {bUsePreviewValueAsDefault=true, inputName=pow, inputType=FunctionInput_Scalar, sortPriority=2, previewValue=<1 0 0 1>}
  FunctionOutput(None=Power(Base=Subtract(A=@n4, B=@n1), Exp=@n5))
 ]}
