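# Material Gym sample material that demonstrates the MF_InverseProportional grayscale transform.
# UVs are remapped around the centre and made absolute, then their two channels are summed into a diamond-shaped distance gradient.
# The function transforms that gradient using scale and pow scalar parameters, with its invert input fed by a static bool.
# The result multiplies a Color parameter, magenta by default, to tint the pattern.
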Material /Game/Global/Materials/MaterialGym/Materials/MaterialElements/GrayscaleTransforms/InverseProportional/M_InverseProportional
graph {
 properties {bAutomaticallySetUsageInEditor=false, usedShadingModels=Default Lit}
 nodes [
  Constant()
  n1 = Abs(None=ConstantBiasScale(None=TextureCoordinate()) {bias=-0.5, scale=2})
  n2 = MaterialFunctionCall!(
   value=Add(A=ComponentMask(None=@n1) {r=true}, B=ComponentMask(None=@n1) {g=true})
   scale=ScalarParameter() {defaultValue=0.03, parameterName="scale (S)"}
   invert value=StaticBool()
   pow=ScalarParameter() {defaultValue=1.25, parameterName="pow (S)"}
  ) {materialFunction=&/MF_InverseProportional}
  Multiply(A=VectorParameter() {parameterName=Color, defaultValue=<rgba 1 0 1 1>}.RGB, B=@n2.Result)
 ]
}
